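# Sound cue for the first-stage (ReloadA) reload foley of the Athena Easter launcher, routed to the weapon foley sound class with reload attenuation.
# A top mixer layers an immediate grenade launcher shell-insert wave with a delayed egg launcher canister-insert wave at slightly lower volume.
# The canister layer follows a fixed sub-second delay and passes through a pitch-varying modulator and a secondary mixer whose second input is empty.
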SoundCue /Game/Sounds/Fort_Weapons/Launchers/Easter/Launcher_Easter_ReloadA_Athena_Cue
root=Mixer_0
nodes [
 {id=Delay_0, children [Mixer_1], properties {delayMax=0.8, delayMin=0.8}}
 {id=Mixer_0, children [WavePlayer_2,Delay_0], properties {inputVolume [0.8,0.7]}}
 {id=Mixer_1, children [Modulator_0,null], properties {inputVolume [1,1]}}
 {id=Modulator_0, children [WavePlayer_0], properties {pitchMax=1.1}}
 {id=WavePlayer_0, properties {soundWaveAssetPtr=/Game/Sounds/Fort_Weapons/Launchers/Egg_Launcher/Egg_Launcher_Reload_Canister_Insert_01}}
 {id=WavePlayer_2, properties {soundWaveAssetPtr=/Game/Sounds/Fort_Weapons/Launchers/grenadelauncher_reload_insert_shell}}
]
properties {
 attenuationSettings=&Fort_Weapon_Reload_Att
 maxDistance=2250
 soundClassObject=&Fort_WeaponFoley
 volumeMultiplier=0.752
}
