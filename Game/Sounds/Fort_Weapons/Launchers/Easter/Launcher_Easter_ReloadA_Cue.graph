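# Sound cue for the Easter launcher's ReloadA sound, reusing the grenade launcher's shell-insert reload wave.
# A single wave player is the root, with no randomization or mixing.
# It uses the shared weapon reload attenuation asset and the weapon foley sound class, with slightly reduced volume.
# Overridden concurrency allows one instance at a time, using the stop-farthest-then-prevent-new rule.
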SoundCue /Game/Sounds/Fort_Weapons/Launchers/Easter/Launcher_Easter_ReloadA_Cue
root=WavePlayer_2
nodes [{id=WavePlayer_2, properties {soundWaveAssetPtr=/Game/Sounds/Fort_Weapons/Launchers/grenadelauncher_reload_insert_shell}}]
properties {
 attenuationSettings=&Fort_Weapon_Reload_Att
 bOverrideConcurrency=true
 maxDistance=2250
 soundClassObject=&Fort_WeaponFoley
 volumeMultiplier=0.902
 attenuationOverrides {falloffDistance=5000, attenuationShapeExtents {x=1000}}
 concurrencyOverrides {maxCount=1, resolutionRule=StopFarthestThenPreventNew}
}
