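# Sound cue for a reload stage ("ReloadB") of the Athena Easter launcher weapon.
# A mixer layers the base reload wave with a louder canister-insert sparkle from the Egg Launcher sounds, run through a modulator for pitch variation.
# It uses weapon reload attenuation and the weapon foley sound class, with overridden concurrency allowing one instance per owner.
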SoundCue /Game/Sounds/Fort_Weapons/Launchers/Easter/Launcher_Easter_ReloadB_Athena_Cue
root=Mixer_0
nodes [
 {id=Mixer_0, children [WavePlayer_1,Modulator_1], properties {inputVolume [1,1.7]}}
 {id=Modulator_1, children [WavePlayer_0], properties {pitchMax=1.2}}
 {
  id=WavePlayer_0
  properties {soundWaveAssetPtr=/Game/Sounds/Fort_Weapons/Launchers/Egg_Launcher/Egg_Launcher_Reload_Canister_Insert_Sparkle_01}
 }
 {id=WavePlayer_1, properties {soundWaveAssetPtr=&/Launcher_Easter_Reload_01_B}}
]
properties {
 attenuationSettings=&Fort_Weapon_Reload_Att
 bOverrideConcurrency=true
 maxDistance=2250
 soundClassObject=&Fort_WeaponFoley
 volumeMultiplier=0.501
 concurrencyOverrides {bLimitToOwner=true, maxCount=1, resolutionRule=StopFarthestThenPreventNew}
}
