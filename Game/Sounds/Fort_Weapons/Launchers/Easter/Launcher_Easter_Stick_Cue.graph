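# Sound cue for the stick sound of the Easter launcher weapon, handled as a bullet-impact style effect.
# A random node picks one of three equally weighted wave variations and feeds a modulator that slightly varies pitch and boosts volume.
# It uses the shared bullet-impact attenuation, the impacts sound class and the rifle bullet impact concurrency group to limit overlapping instances.
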SoundCue /Game/Sounds/Fort_Weapons/Launchers/Easter/Launcher_Easter_Stick_Cue
root=Modulator_6
nodes [
 {id=Modulator_6, children [Random_1], properties {pitchMax=1.1, pitchMin=0.9, volumeMax=1.3, volumeMin=1.2}}
 {id=Random_1, children [WavePlayer_5,WavePlayer_6,WavePlayer_7], properties {weights [1,1,1]}}
 {id=WavePlayer_5, properties {soundWaveAssetPtr=&/Launcher_Easter_Stick_01}}
 {id=WavePlayer_6, properties {soundWaveAssetPtr=&/Launcher_Easter_Stick_02}}
 {id=WavePlayer_7, properties {soundWaveAssetPtr=&/Launcher_Easter_Stick_03}}
]
properties {
 attenuationSettings=/Game/Sounds/Attenuation/Fort_Bullet_Impact_Att
 maxDistance=5300
 soundClassObject=&Fort_Impacts
 volumeMultiplier=1
 concurrencySet [&RifleBulletImpact_Conc]
}
